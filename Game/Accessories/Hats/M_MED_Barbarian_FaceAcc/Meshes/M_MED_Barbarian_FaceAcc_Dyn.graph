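# Physics asset for the M_MED_Barbarian_FaceAcc hat accessory, defining bodies and joints for its dynamic hair and ponytail bones.
# Two four-capsule ponytail chains and three two-box hair chains hang from the head, with tight twist limits and narrower swing on the hair than the ponytails.
# Head, neck, upper spine, clavicle and upper-arm shapes form the body volume around them, linked by constraints with free swing and twist.
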PhysicsAsset /Game/Accessories/Hats/M_MED_Barbarian_FaceAcc/Meshes/M_MED_Barbarian_FaceAcc_Dyn
bodies {
 clavicle_l [{shapeType=Capsule, length=9.1113, radius=9.3757, center=<6.5353 -1.2986 0.096>, rotation=<rot 74.9997 0 0>}]
 clavicle_r [{shapeType=Capsule, length=9.1113, radius=9.3757, center=<-6.5353 1.2986 -0.096>, rotation=<rot 74.9997 0 0>}]
 dyn_hair_01 [{shapeType=Box, extentX=5.9004, extentY=1.8791, extentZ=7.5124, center=<3.2348 0 -0.0001>}]
 dyn_hair_01_l [{shapeType=Box, extentX=5.9004, extentY=1.8791, extentZ=7.5124, center=<3.2348 0 -0.0001>}]
 dyn_hair_01_r [{shapeType=Box, extentX=5.9004, extentY=1.8791, extentZ=7.5124, center=<-3.2348 0 0>}]
 dyn_hair_02 [{shapeType=Box, extentX=4.0879, extentY=1.8791, extentZ=7.5124, center=<2.7362 0 -0.0001>}]
 dyn_hair_02_l [{shapeType=Box, extentX=3.9004, extentY=1.8791, extentZ=7.5124, center=<2.2105 0 -0.0002>}]
 dyn_hair_02_r [{shapeType=Box, extentX=3.9004, extentY=1.8791, extentZ=7.5124, center=<-2.2105 0 0>}]
 dyn_ponytail_l_01 [{shapeType=Capsule, length=6.1562, radius=2, center=<4.5388 0 0>, rotation=<rot 90 0 0>}]
 dyn_ponytail_l_02 [{shapeType=Capsule, length=6.7188, radius=2.4375, center=<5.11 0.0001 0>, rotation=<rot 90 0 0>}]
 dyn_ponytail_l_03 [{shapeType=Capsule, length=1.4375, radius=2.5625, center=<2.1293 0 0>, rotation=<rot 90 0 0>}]
 dyn_ponytail_l_04 [{shapeType=Capsule, length=3.5625, radius=3, center=<3.3101 0 0.0004>, rotation=<rot 90 0 0>}]
 dyn_ponytail_r_01 [{shapeType=Capsule, length=6.1562, radius=2, center=<-4.5388 0 0>, rotation=<rot 90 0 0>}]
 dyn_ponytail_r_02 [{shapeType=Capsule, length=6.7188, radius=2.4375, center=<-5.11 0 0>, rotation=<rot 90 0 0>}]
 dyn_ponytail_r_03 [{shapeType=Capsule, length=1.4375, radius=2.5625, center=<-2.1293 0 0>, rotation=<rot 90 0 0>}]
 dyn_ponytail_r_04 [{shapeType=Capsule, length=3.5625, radius=3, center=<-3.6568 0 0>, rotation=<rot 90 0 0>}]
 head [
  {shapeType=Capsule, length=2.85, radius=5.7337, center=<1.5439 3.4489 0.0001>, rotation=<rot -89.5256 -89.7443 90.2186>}
  {shapeType=Capsule, length=7.4331, radius=2.7748, center=<1.1568 -0.1738 -6.9765>, rotation=<rot -84.9996 -0.0001 10.0014>}
  {shapeType=Capsule, length=7.4331, radius=2.7748, center=<1.1568 -0.1738 6.9765>, rotation=<rot 84.9996 -0.0001 -10.0014>}
 ]
 neck_02 [{shapeType=Capsule, length=10.9843, radius=4.6791, rotation=<rot 0 -89.9999 -89.9996>}]
 spine_05 [
  {shapeType=Capsule, length=15.125, radius=14.8488, center=<-3.6854 -2.3157 0>}
  {shapeType=Capsule, length=16.75, radius=2.7188, center=<-4.1838 14.3014 -7.5248>, rotation=<rot 39.9997 0 15.0013>}
  {shapeType=Capsule, length=25, radius=11, shapeName=spine_05_capsule, center=<11.3738 8.3223 -21.2899>, rotation=<rot 87 180 -150>}
  {shapeType=Box, extentX=40.125, extentY=29.9375, extentZ=11.1875, center=<8.365 7.0907 19.1305>, rotation=<rot -6.9142 -29.0006 -9.4027>}
  {shapeType=Box, extentX=54.875, extentY=7.5, extentZ=49.0625, center=<19.508 -6.001 0>, rotation=<rot 0 5.0001 0>}
 ]
 upperarm_l [{shapeType=Capsule, length=18.5848, radius=6.3549, center=<11.4355 0 0>, rotation=<rot 90 0 0>}]
 upperarm_r [{shapeType=Capsule, length=18.5848, radius=6.3549, center=<-11.4355 0 0>, rotation=<rot 90 0 0>}]
}
constraints [
 {bone1Name=dyn_ponytail_l_01, bone2Name=head, swing1LimitDegrees=45, swing2LimitDegrees=45, twistLimitDegrees=5}
 {bone1Name=dyn_ponytail_l_02, bone2Name=dyn_ponytail_l_01, swing1LimitDegrees=45, swing2LimitDegrees=45, twistLimitDegrees=5}
 {bone1Name=dyn_ponytail_l_03, bone2Name=dyn_ponytail_l_02, swing1LimitDegrees=45, swing2LimitDegrees=45, twistLimitDegrees=5}
 {bone1Name=dyn_ponytail_l_04, bone2Name=dyn_ponytail_l_03, swing1LimitDegrees=45, swing2LimitDegrees=45, twistLimitDegrees=5}
 {bone1Name=dyn_ponytail_r_01, bone2Name=head, swing1LimitDegrees=45, swing2LimitDegrees=45, twistLimitDegrees=5}
 {bone1Name=dyn_ponytail_r_02, bone2Name=dyn_ponytail_r_01, swing1LimitDegrees=45, swing2LimitDegrees=45, twistLimitDegrees=5}
 {bone1Name=dyn_ponytail_r_03, bone2Name=dyn_ponytail_r_02, swing1LimitDegrees=45, swing2LimitDegrees=45, twistLimitDegrees=5}
 {bone1Name=dyn_ponytail_r_04, bone2Name=dyn_ponytail_r_03, swing1LimitDegrees=45, swing2LimitDegrees=45, twistLimitDegrees=5}
 {bone1Name=dyn_hair_01, bone2Name=head, swing1LimitDegrees=20, swing2LimitDegrees=20, twistLimitDegrees=5}
 {bone1Name=dyn_hair_02, bone2Name=dyn_hair_01, swing1LimitDegrees=20, swing2LimitDegrees=20, twistLimitDegrees=5}
 {bone1Name=dyn_hair_01_l, bone2Name=head, swing1LimitDegrees=20, swing2LimitDegrees=20, twistLimitDegrees=5}
 {bone1Name=dyn_hair_02_l, bone2Name=dyn_hair_01_l, swing1LimitDegrees=20, swing2LimitDegrees=20, twistLimitDegrees=5}
 {bone1Name=dyn_hair_01_r, bone2Name=head, swing1LimitDegrees=20, swing2LimitDegrees=20, twistLimitDegrees=5}
 {bone1Name=dyn_hair_02_r, bone2Name=dyn_hair_01_r, swing1LimitDegrees=20, swing2LimitDegrees=20, twistLimitDegrees=5}
 {
  bone1Name=head
  bone2Name=neck_02
  swing1LimitDegrees=45
  swing1Motion=Free
  swing2LimitDegrees=45
  swing2Motion=Free
  twistLimitDegrees=45
  twistMotion=Free
 }
 {
  bone1Name=clavicle_l
  bone2Name=spine_05
  swing1LimitDegrees=45
  swing1Motion=Free
  swing2LimitDegrees=45
  swing2Motion=Free
  twistLimitDegrees=45
  twistMotion=Free
 }
 {
  bone1Name=upperarm_l
  bone2Name=clavicle_l
  swing1LimitDegrees=45
  swing1Motion=Free
  swing2LimitDegrees=45
  swing2Motion=Free
  twistLimitDegrees=45
  twistMotion=Free
 }
 {
  bone1Name=clavicle_r
  bone2Name=spine_05
  swing1LimitDegrees=45
  swing1Motion=Free
  swing2LimitDegrees=45
  swing2Motion=Free
  twistLimitDegrees=45
  twistMotion=Free
 }
 {
  bone1Name=upperarm_r
  bone2Name=clavicle_r
  swing1LimitDegrees=45
  swing1Motion=Free
  swing2LimitDegrees=45
  swing2Motion=Free
  twistLimitDegrees=45
  twistMotion=Free
 }
 {
  bone1Name=neck_02
  bone2Name=spine_05
  swing1LimitDegrees=45
  swing1Motion=Free
  swing2LimitDegrees=45
  swing2Motion=Free
  twistLimitDegrees=45
  twistMotion=Free
 }
]
